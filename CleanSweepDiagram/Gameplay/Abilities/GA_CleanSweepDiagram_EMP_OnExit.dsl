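; Gameplay ability for an EMP triggered on exit, as its name indicates, that fires an explosion gameplay cue and applies gameplay effects to targets.
; On activation it calls the parent, executes the EMP cue on the avatar, then runs a DoTargetSelection task named EMPTargeting and applies the matching effect container.
; A second selection, EMPTargetingWorld, is nested in the first task's Targeted callback, so it runs only after that fires and applies its own tagged container.
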(event Ability|EventActivateAbility
  (|CallParentFunction)
  ;[1 0 0.9595] Explode GC
  (GameplayCue|ExecuteGameplayCue (Ability|GetAvatarActorfromActorInfo) (Variables|Default|GetEMP%20Gameplay%20Cue%20Tag))
  ;[0 0.3136 1] Gameplay Effects Container Targeting
  (bind (_asynctaskproxy _targetdata _applicationtag) (Ability|Tasks|DoTargetSelection "EMPTargeting" (Variables|Default|GetIn%20Application%20Tag) true)
    (:Targeted
      (Ability|ApplyGameplayEffectContainer _targetdata _applicationtag)
      (bind (_asynctaskproxy_2 _targetdata_2 _applicationtag_2) (Ability|Tasks|DoTargetSelection "EMPTargetingWorld" (Variables|Default|GetIn%20Application%20Tag_1) true)
        (:Targeted
          (Ability|ApplyGameplayEffectContainer _targetdata_2 _applicationtag_2))))))

(event Ability|EventOnEndAbility
  (params (bWasCancelled :type Boolean)))
