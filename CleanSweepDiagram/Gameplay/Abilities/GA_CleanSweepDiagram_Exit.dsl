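; Gameplay ability that runs the exit from the CleanSweepDiagram state on the player.
; It optionally holsters the weapon, waits briefly, then does nothing further if blocking tags are present.
; With the expected weapon it clears HUD keys and plays a recovery montage, whose trigger applies the EMP effect; otherwise the effect applies immediately.
; The montage path later removes the granting effect and ends the ability, and ability end clears the active-camo cue.
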(event Ability|EventActivateAbility
  (bind _playerpawn (Variables|Default|GetPlayerPawn))
  (bind _currentweapon (Variables|/Script/FortniteGame.FortPawn|GetCurrentWeapon _playerpawn))
  (bind _asplayer_pawn_athena (Utilities|Casting|CastToPlayerPawn_Athena (Ability|GetAvatarActorfromActorInfo)))
  (Variables|Default|SetPlayerPawn _asplayer_pawn_athena)
  (if (Variables|Default|GetHolsterWeapon)
    (Weapon|SetHolsterWeapon (Variables|Default|GetPlayerPawn) 0 true false (Variables|Default|GetSpookyHolster)))
  (Utilities|FlowControl|Delay 0.2) ; Slight delay so weapon is not in hand when montage starts.
  (if (GameplayTags|HasAnyMatchingGameplayTags (Variables|Default|GetPlayerPawn) (Variables|Default|GetTag%20Container))
    _
    (elif (== (Variables|Default|GetWeaponToCheckPlayAnim) (Utilities|GetClass _currentweapon))
      (bind _asb_clean_sweep_diagram (Utilities|Casting|CastToB_CleanSweepDiagram _currentweapon))
      (Class|BCleanSweepDiagram|SetHudKeys _asb_clean_sweep_diagram false)
      (Ability|Tasks|PlayMontageandWaitforTrigger "None" (Variables|Default|GetExitRecovery)
        (:then
          (Utilities|FlowControl|Delay (ScalableFloat|GetValueatLevel (Variables|Default|GetDelayExit)))
          (Ability|RemoveGrantedbyEffect)
          (Ability|EndAbility)
          (Ability|RemoveGameplayCueFromOwner (Variables|Default|GetTag_GCNL_ActiveCamo)))
        (:Triggered
          (CallFunction|ExitMontageTriggeredEvent)))
      (else
        (CallFunction|ExitMontageTriggeredEvent)))))

(event Ability|EventOnEndAbility
  (params (bWasCancelled :type Boolean))
  (|CallParentFunction bWasCancelled)
  (Ability|RemoveGameplayCueFromOwner (Variables|Default|GetTag_GCNL_ActiveCamo)))

(event EventExitMontageTriggeredEvent
  ;[0 0.2562 1] Anim Montage waits for trigger, which calls this in the parent. Applies this GE to start the EMP GA
  (Ability|ApplyGameplayEffectToOwner "/CleanSweepDiagram/Gameplay/GameplayEffects/GE_CleanSweepDiagram_Exit_EMP.GE_CleanSweepDiagram_Exit_EMP_C"))

(fn CanApplyHolster
  (params (HolsterId :type Name%20%28by%20ref%29 :default "None"))
  (return false))
