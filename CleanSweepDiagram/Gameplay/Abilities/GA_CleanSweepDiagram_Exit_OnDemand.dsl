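; Gameplay ability for the on-demand exit from CleanSweepDiagram active camo, which removes the camo and fires an exit EMP.
; On activation it adds loose tags blocking the standard exit ability, waits, removes the camo cue and granting effect, then clears the tags after a short delay.
; Ending the ability repeats the removal and delayed tag clearing.
; The parent's exit montage trigger applies the EMP effect, and holstering is always refused.
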(event Ability|EventActivateAbility
  (|CallParentFunction)
  ;[0 0.06352 1] Blocking tag for standard exit GA
  (Ability|GameplayEffect|AddLooseGameplayTags (Ability|GetAvatarActorfromActorInfo) (Variables|Default|GetBlock%20Standard%20Exit%20GA%20Gameplay%20Tags))
  (Utilities|FlowControl|Delay (Variables|Default|GetDelayBeforeTriggerRemovalAndEMP))
  (Ability|RemoveGameplayCueFromOwner (Variables|Default|GetTag_GCNL_ActiveCamo))
  (Ability|RemoveGameplayEffectFromOwnerWithGrantedTags (Variables|Default|GetTag_ActiveCamo_Active)) ; Removes granting effect, ending ability
  ;[0 0.06352 1] Small delay to ensure we block the standard exit GA
  (Utilities|FlowControl|Delay 0.5)
  (Ability|GameplayEffect|RemoveLooseGameplayTags (Ability|GetAvatarActorfromActorInfo) (Variables|Default|GetBlock%20Standard%20Exit%20GA%20Gameplay%20Tags)))

(event Ability|EventOnEndAbility
  (params (bWasCancelled :type Boolean))
  (|CallParentFunction bWasCancelled)
  (Ability|RemoveGameplayCueFromOwner (Variables|Default|GetTag_GCNL_ActiveCamo))
  (Ability|RemoveGameplayEffectFromOwnerWithGrantedTags (Variables|Default|GetTag_ActiveCamo_Active)) ; Removes granting effect, ending ability
  ;[0 0.06352 1] Small delay to ensure we block the standard exit GA
  (Utilities|FlowControl|Delay 0.5)
  (Ability|GameplayEffect|RemoveLooseGameplayTags (Ability|GetAvatarActorfromActorInfo) (Variables|Default|GetBlock%20Standard%20Exit%20GA%20Gameplay%20Tags)))

(event EventExitMontageTriggeredEvent
  ;[0 0.2562 1] Anim Montage waits for trigger, which calls this in the parent. Applies this GE to start the EMP GA
  (Ability|ApplyGameplayEffectToOwner "/CleanSweepDiagram/Gameplay/GameplayEffects/GE_CleanSweepDiagram_Exit_EMP.GE_CleanSweepDiagram_Exit_EMP_C"))

(fn CanApplyHolster
  (params (HolsterId :type Name%20%28by%20ref%29 :default "None"))
  (return false))
